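; Gameplay ability that handles catching the thrown Ancient Weapon boomerang.
; With the weapon equipped and no skydiving, it blocks recall, plays a movement-dependent catch montage and cue, and updates weapon visibility and emissives.
; Weapon switches or a failsafe timer end it; otherwise, an unequipped cue plays without montage.
; Ending clears recall and throw-block tags, restores the throw prompt, and forces a net update because the anim graph reads the tag.
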(event Custom|FailsafeEnd
  ;[1 1 1] If the ability is running for too long kill it.
  (Utilities|FlowControl|DoOnce false
    (:Completed
      (Utilities|Time|ClearandInvalidateTimerbyHandle (Variables|CatchVars|GetTimer_FailsafeEnd))
      (Ability|EndAbility))))

(event Ability|EventActivateAbility
  (bind _returnvalue (Ability|GetAvatarActorfromActorInfo))
  ;[1 1 1] Save Out Player
  (Variables|CatchVars|SetMontageToPlay (select (Math|Boolean|ANDBoolean (Variables|/Script/Engine.Character|GetbIsCrouched (Utilities|Casting|CastToFortPlayerPawn (Ability|GetAvatarActorfromActorInfo))) (AI|Components|NavMovement|IsMovingonGround (Variables|/Script/Engine.Character|GetCharacterMovement (Utilities|Casting|CastToFortPlayerPawn (Ability|GetAvatarActorfromActorInfo)))) (>= (Math|Vector|VectorLength (Transformation|GetVelocity (Utilities|Casting|CastToFortPlayerPawn (Ability|GetAvatarActorfromActorInfo)))) 1e+01) (not (PawnStats|IsSprinting (Utilities|Casting|CastToFortPlayerPawn (Ability|GetAvatarActorfromActorInfo))))) (Variables|Default|GetCatchInMotionCMM) (Variables|Default|GetNormalCatchCMM)))
  (Utilities|IsValid _returnvalue
    (:Is%20Valid
      (bind _asfort_player_pawn (Utilities|Casting|CastToFortPlayerPawn _returnvalue)
        (:then
          (Variables|CatchVars|SetPlayerPawn _asfort_player_pawn)
          ;[1 1 1] Check if we have the weapon out and we are not skydiving
          (if (and (Utilities|Equal(Class) (Utilities|GetClass (Variables|/Script/FortniteGame.FortPawn|GetCurrentWeapon (Variables|CatchVars|GetPlayerPawn))) (Variables|CatchVars|GetWeaponToCheckPlayAnim)) (not (Parachute|IsSkydiving (Variables|CatchVars|GetPlayerPawn))))
            ;[0 0.6251 1] Tag added to block recall use during catch GA
            (Ability|GameplayEffect|AddLooseGameplayTags (Variables|CatchVars|GetPlayerPawn) (Variables|Default|GetCatch%20In%20Progress%20Gameplay%20Tags))
            ;[1 1 1] Play Montage and then end ability when finished. Make sure we do once.
            (Ability|Tasks|PlayMontageandWaitforTrigger "None" (Variables|CatchVars|GetMontageToPlay)
              (:then
                (CallFunction|SetWeaponMeshVisibility)
                (Variables|Default|SetStartedMontage true)
                (Ability|Tasks|WaitGameplayEvent "(TagName=\"Event.Weapons.WeaponSwitch\")"
                  (:then
                    ;[1 1 1] Send GC when we start the montage
                    (Ability|ExecuteGameplayCueOnOwner (Variables|CatchVars|GetT_CatchGC))
                    ;[1 1 1] If the ability is running for too long kill it.
                    (bind _returnvalue_2 (Utilities|Time|SetTimerbyEvent (delegate Custom|FailsafeEnd) (Variables|CatchVars|GetFailsafeDelay)))
                    (Variables|CatchVars|SetTimer_FailsafeEnd _returnvalue_2)
                    (bind _asb_chrono_weap_ancient_weapon_thrown (Utilities|Casting|CastToB_Chrono_Weap_AncientWeapon_Thrown (Variables|/Script/FortniteGame.FortPawn|GetCurrentWeapon (Variables|CatchVars|GetPlayerPawn))))
                    (Class|BChronoWeapAncientWeaponThrown|GauntleEmissiveonCatch _asb_chrono_weap_ancient_weapon_thrown false)
                    (Class|BChronoWeapAncientWeaponThrown|BoomerangEmissiveonCooldown _asb_chrono_weap_ancient_weapon_thrown))
                  (:EventReceived
                    ;[1 1 1] Stop montage if it is playing
                    (Animation|StopAnimMontage (Variables|CatchVars|GetPlayerPawn) (Variables|CatchVars|GetMontageToPlay))
                    (Development|PrintString "Stop catch  montage")
                    (Ability|EndAbility))))
              (:Cancelled
                _)
              (:Completed
                _))
            (if (not (Networking|IsServer))
              (Utilities|FlowControl|DoOnce false
                (:Completed
                  (Ability|EndAbility))))
            (else
              ;[1 1 1] Check if we did not have the weapon out just play the GC and end.
              (Development|PrintString "PlayNoCatchAnim")
              (Ability|ExecuteGameplayCueOnOwner (Variables|CatchVars|GetT_Catch_NotEquippedGC))
              (CallFunction|SetWeaponMeshVisibility)
              (Ability|EndAbility))))
        (:CastFailed
          _)))
    (:Is%20Not%20Valid
      _))
  (Ability|EndAbility))

(event Ability|EventOnEndAbility
  (params (bWasCancelled :type Boolean))
  ;[1 1 1] End the ability
  (Utilities|FlowControl|DoOnce false
    (:Completed
      ;[0 0.6251 1] Remove tag to block recall use during catch GA
      (Ability|GameplayEffect|RemoveLooseGameplayTags (Variables|CatchVars|GetPlayerPawn) (Variables|Default|GetCatch%20In%20Progress%20Gameplay%20Tags))
      (bind _asb_chrono_weap_ancient_weapon_thrown (Utilities|Casting|CastToB_Chrono_Weap_AncientWeapon_Thrown (Variables|/Script/FortniteGame.FortPawn|GetCurrentWeapon (Variables|CatchVars|GetPlayerPawn)))
        (:then
          (Class|BChronoWeapAncientWeaponThrown|ToggleBoomerangThrowPrompt _asb_chrono_weap_ancient_weapon_thrown true))
        (:CastFailed
          _))
      ;[1 1 1] Unhide and reequip. If it is not the thrown weapon play the equip montage.
      (Utilities|IsValid (Variables|CatchVars|GetPlayerPawn)
        (:Is%20Valid
          (Ability|GameplayEffect|RemoveLooseGameplayTags (Variables|CatchVars|GetPlayerPawn) (Variables|Default|GetAlready%20Recalled%20Gameplay%20Tag))
          (GameplayEffects|RemoveActiveEffectswithGrantedTags (Ability|GetAbilitySystemComponent (Variables|CatchVars|GetPlayerPawn)) (Variables|Default|GetBlock%20Throw%20Tag))
          ;;[1 1 1 1 12] Need to force a net update to get this GE off immediately, because its gameplay tag is referenced by the anim graph.
          (Networking|ForceNetUpdate (Variables|CatchVars|GetPlayerPawn)))
        (:Is%20Not%20Valid
          ;[1 1 1] Try to remove it from the controller if the pawn does not exist.
          (Ability|GameplayEffect|RemoveLooseGameplayTags (Variables|CatchVars|GetPlayerPawn) (Variables|Default|GetAlready%20Recalled%20Gameplay%20Tag))
          (GameplayEffects|RemoveActiveEffectswithGrantedTags (Ability|GetAbilitySystemComponent (Variables|CatchVars|GetPlayerPawn)) (Variables|Default|GetBlock%20Throw%20Tag)))) ; Make Sure it is still valid
    )))

(fn SetWeaponMeshVisibility
  ;[1 1 1] Show Mesh repped for other clients
  (bind _asb_chrono_weap_ancient_weapon_thrown (Utilities|Casting|CastToB_Chrono_Weap_AncientWeapon_Thrown (Variables|/Script/FortniteGame.FortPawn|GetCurrentWeapon (Variables|CatchVars|GetPlayerPawn))))
  (Utilities|FlowControl|Sequence
    (:then_0
      (if (not (and (GameplayTags|HasMatchingGameplayTag (Variables|CatchVars|GetPlayerPawn) (Variables|Default|GetAncientWeaponSprintTag)) (not (GameplayTags|HasMatchingGameplayTag (Variables|CatchVars|GetPlayerPawn) (Variables|Default|GetAncientWeaponSlideTag)))))
        (Ability|HasAuthority
          (:True
            (Variables|/AncientWeapon/Gameplay/Actors/B_Chrono_Weap_AncientWeapon_Thrown.B_Chrono_Weap_AncientWeapon_Thrown_C|SetbOnRepWeaponMeshVisible true _asb_chrono_weap_ancient_weapon_thrown)))))
    (:then_1
      (return))))
